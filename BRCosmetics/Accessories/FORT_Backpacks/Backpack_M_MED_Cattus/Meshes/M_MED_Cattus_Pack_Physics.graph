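# Physics asset for the M_MED_Cattus backpack cosmetic, giving its flag bone chain physics bodies and joints.
# The master flag bone carries a small capsule and a large slightly rotated box, and four progressively smaller boxes chain from it through dyn_flag1 to dyn_flag4.
# Angular limits are uniform between the flag segments, while the joint to the master bone has much tighter second-swing and twist limits.
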PhysicsAsset /BRCosmetics/Accessories/FORT_Backpacks/Backpack_M_MED_Cattus/Meshes/M_MED_Cattus_Pack_Physics
bodies {
 dyn_flag1 [{shapeType=Box, extentX=2, extentY=1, extentZ=4.5, center=<1 0 0>}]
 dyn_flag2 [{shapeType=Box, extentX=2, extentY=1, extentZ=4, center=<1 0 0>}]
 dyn_flag3 [{shapeType=Box, extentX=1.7, extentY=1, extentZ=2.5, center=<0.85 0 0>}]
 dyn_flag4 [{shapeType=Box, extentX=1, extentY=1, extentZ=2, center=<0.5 0 0>}]
 dyn_mstr_flag [
  {shapeType=Capsule, length=0.505, radius=0.505, rotation=<rot 0 -90 0>}
  {shapeType=Box, extentX=19, extentY=1.5, extentZ=14.5, center=<2.5193 1.6515 0>, rotation=<rot 0 2.0698 0>}
 ]
}
constraints [
 {bone1Name=dyn_flag2, bone2Name=dyn_flag1, swing1LimitDegrees=50, swing2LimitDegrees=45, twistLimitDegrees=15}
 {bone1Name=dyn_flag3, bone2Name=dyn_flag2, swing1LimitDegrees=50, swing2LimitDegrees=45, twistLimitDegrees=15}
 {bone1Name=dyn_flag4, bone2Name=dyn_flag3, swing1LimitDegrees=50, swing2LimitDegrees=45, twistLimitDegrees=15}
 {bone1Name=dyn_flag1, bone2Name=dyn_mstr_flag, swing1LimitDegrees=50, swing2LimitDegrees=5, twistLimitDegrees=1}
]
