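# Animation Blueprint for the M_MED_Cattus backpack, adding secondary physics and a procedural floating motion to the pack's skeleton.
# A RigidBody node simulates the pack's physics asset in the space of the pack bone, with overridden gravity and damped component motion.
# Each update, the event graph builds an offset from game-time sine waves with different frequency and amplitude per axis.
# That offset additively translates the dyn_islandBase bone in bone space after simulation.
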AnimBlueprint /BRCosmetics/Accessories/FORT_Backpacks/Backpack_M_MED_Cattus/Meshes/M_MED_Cattus_Pack_AnimBP
target_skeleton=&/M_MED_Cattus_Pack_Skeleton
graphs [
 AnimGraph {nodes [
   n1 = RigidBody(ComponentPose=Local To Component(LocalPose=Input Pose())) {
    bClampLinearTranslationLimitToRefPose=true
    bOverrideWorldGravity=true
    overridePhysicsAsset=&/M_MED_Cattus_Pack_Physics
    simulationSpace=BaseBoneSpace
    alphaBoolBlend {blendOption=Linear}
    baseBoneRef {boneName=pack}
    componentAppliedLinearAccClamp=<5000 5000 5000>
    componentLinearAccScale=<0.5 0.5 0.5>
    componentLinearVelScale=<0.5 0.5 0.5>
    overrideWorldGravity {z=-980}
   }
   n2 = "Transform (Modify) Bone - Bone: dyn_islandBase"(ComponentPose=@n1, Alpha=1, Translation=Get islandFloat() -> (islandFloat:Vector), Rotation="0,0,0", Scale="1,1,1") {
    rotationMode=BMM_Additive
    rotationSpace=BCS_BoneSpace
    translationMode=BMM_Additive
    translationSpace=BCS_BoneSpace
    alphaBoolBlend {blendOption=Linear}
    boneToModify {boneName=dyn_islandBase}
   }
   Output Pose(Result=Component To Local(ComponentPose=@n2)) {layerGroup=None}
  ]}
 EventGraph {nodes [
   Event BlueprintUpdateAnimation()
   n1 = GetGameTimeInSeconds()
   n2 = "Sin (Radians)"(A=float * float(A=float * float(A=1, B=1.5), B=@n1))
   n3 = "Sin (Radians)"(A=float * float(A=float * float(A=1, B=2), B=@n1))
   n4 = "Sin (Radians)"(A=float * float(A=float * float(A=1, B=1), B=@n1))
   > Set islandFloat(islandFloat:Vector=MakeVector(X=float * float(A=@n2, B=0.3), Y=float * float(A=@n4, B=0.6), Z=float * float(A=@n3, B=0.25))) -> (Output_Get:Vector)
   TryGetPawnOwner()
  ]}
]
variables [islandFloat {type=Vector}]
